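# Physics asset for the DrinkUp Keytar instrument mesh in Sparks cosmetics, giving its drink cans and handle simple collision bodies for physics.
# Each of the five cans (one slide can, four static cans) has a similarly sized capsule, and the handle has an offset capsule.
# Constraints link every can directly to the handle in a star layout, with identical tight swing and twist limits of a few degrees.
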PhysicsAsset /SparksCosmetics/Instrument/Keytar/DrinkUp/Meshes/DrinkUp_Keytar_Physics
bodies {
 dyn_slide_can [{shapeType=Capsule, length=5.5, radius=3, shapeName=dyn_slide_can_capsule, rotation=<rot 0 -90 0>}]
 dyn_static_can_01 [{
   shapeType=Capsule
   length=5.6933
   radius=3.0439
   shapeName=dyn_static_can_01_capsule
   center=<-0.0043 -0.0123 0.0599>
   rotation=<rot 0 89.8599 -89.3199>
  }]
 dyn_static_can_02 [{
   shapeType=Capsule
   length=5.6933
   radius=3.0291
   shapeName=dyn_static_can_02_capsule
   center=<-0.0043 -0.0193 0.0581>
   rotation=<rot 0 89.7806 -89.3411>
  }]
 dyn_static_can_03 [{
   shapeType=Capsule
   length=5.6933
   radius=3.0407
   shapeName=dyn_static_can_03_capsule
   center=<-0.0043 -0.0104 0.0603>
   rotation=<rot 0 89.8816 -89.3157>
  }]
 dyn_static_can_04 [{
   shapeType=Capsule
   length=5.6933
   radius=3.0437
   shapeName=dyn_static_can_04_capsule
   center=<-0.0043 -0.0147 0.0594>
   rotation=<rot 0 89.8329 -89.326>
  }]
 handle [{shapeType=Capsule, length=5, radius=5, shapeName=handle_capsule, center=<0 -22.8496 -12.9818>, rotation=<rot 90 0 0>}]
}
constraints [
 {bone1Name=dyn_static_can_01, bone2Name=handle, swing1LimitDegrees=3, swing2LimitDegrees=3, twistLimitDegrees=3}
 {bone1Name=dyn_static_can_02, bone2Name=handle, swing1LimitDegrees=3, swing2LimitDegrees=3, twistLimitDegrees=3}
 {bone1Name=dyn_static_can_03, bone2Name=handle, swing1LimitDegrees=3, swing2LimitDegrees=3, twistLimitDegrees=3}
 {bone1Name=dyn_static_can_04, bone2Name=handle, swing1LimitDegrees=3, swing2LimitDegrees=3, twistLimitDegrees=3}
 {bone1Name=dyn_slide_can, bone2Name=handle, swing1LimitDegrees=3, swing2LimitDegrees=3, twistLimitDegrees=3}
]
